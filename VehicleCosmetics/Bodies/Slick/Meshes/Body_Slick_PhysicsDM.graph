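# Physics asset for the Slick vehicle body cosmetic, supplying simplified collision for the car mesh.
# It holds a single body on the root bone, shaped as one box sized to the car and offset upward and slightly forward from the origin.
# It has no constraints, so the whole body collides as one rigid volume.
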PhysicsAsset /VehicleCosmetics/Bodies/Slick/Meshes/Body_Slick_PhysicsDM
bodies {root [{shapeType=Box, extentX=444, extentY=216, extentZ=116, shapeName=root_box, center=<30 0 92>}]}
constraints []
